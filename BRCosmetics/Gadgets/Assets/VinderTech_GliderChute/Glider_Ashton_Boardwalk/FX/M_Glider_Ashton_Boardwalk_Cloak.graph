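# Translucent lit cloak material for the Ashton Boardwalk glider, fading it through a grid of square cells with refraction.
# Tiled object-relative positions form box-distance masks on three planes, blended by local normal and scaled by FadeControl or a debug sine, then multiplied by a custom-depth test for opacity.
# Diffuse, normal and mask textures supply the surface, with view-angle spec tint, tinted and panning emissive, and quality/shading-path fallbacks for normal and refraction.
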Material /BRCosmetics/Gadgets/Assets/VinderTech_GliderChute/Glider_Ashton_Boardwalk/FX/M_Glider_Ashton_Boardwalk_Cloak
graph {
 properties {
  allowTranslucentCustomDepthWrites=true
  bUseMaterialAttributes=true
  bUsedWithMeshDeformer=true
  bUsedWithSkeletalMesh=true
  blendMode=BLEND_Translucent
  pixelDepthOffsetMode=PDOM_Legacy
  refractionMethod=RM_IndexOfRefraction
  translucencyLightingMode=TLM_VolumetricPerVertexDirectional
  usedShadingModels=Default Lit
 }
 nodes [
  n1 = TextureSample() {
   samplerSource=SSM_Wrap_WorldGroupSettings
   texture=/BRCosmetics/Gadgets/Assets/VinderTech_GliderChute/Glider_Ashton_Boardwalk/Textures/T_Ashton_Boardwalk_D
  }
  n2 = VectorParameter() {parameterName=Base Color, defaultValue=<rgba 1 1 1 0>}
  n3 = Multiply(A=@n2.RGB, B=VectorParameter() {parameterName=SpecColor, defaultValue=<rgba 1 1 1 0>}.RGB)
  n4 = TextureSample() {
   samplerSource=SSM_Wrap_WorldGroupSettings
   samplerType=SAMPLERTYPE_Normal
   texture=/BRCosmetics/Gadgets/Assets/VinderTech_GliderChute/Glider_Ashton_Boardwalk/Textures/T_Ashton_Boardwalk_N
  }
  n5 = ScalarParameter() {defaultValue=5, parameterName=SpecRange}
  n6 = Constant() {r=1}
  n7 = DotProduct(
   A=Transform(None=@n4.RGB)
   B=ShadingPathSwitch(
    Default=MaterialFunctionCall("FadeSensitivity (optional)"=@n5) {materialFunction=&FadeBasedOnViewAngle}.Result
    Forward=@n6
    Mobile=@n6
   ) {desc=Migrated from FeatureLevelSwitch}
  )
  n8 = TextureSample() {
   samplerSource=SSM_Wrap_WorldGroupSettings
   samplerType=SAMPLERTYPE_Masks
   texture=/BRCosmetics/Gadgets/Assets/VinderTech_GliderChute/Glider_Ashton_Boardwalk/Textures/T_Ashton_Boardwalk_S
  }
  n9 = TextureSample() {
   samplerSource=SSM_Wrap_WorldGroupSettings
   texture=/BRCosmetics/Gadgets/Assets/VinderTech_GliderChute/Glider_Ashton_Boardwalk/Textures/T_Ashton_Boardwalk_E
  }
  n10 = Multiply(A=@n9.RGB, B=VectorParameter() {parameterName=FanEmissive, defaultValue=<rgba 1 1 1 0>}.RGB)
  n11 = WorldPosition()
  n12 = ObjectPositionWS()
  n13 = VectorParameter() {parameterName=Texture Size, defaultValue=<rgba 5 5 5 0>}
  n14 = Divide(
   A=Subtract(
    A=TransformPosition(None=@n11.XYZ) {transformSourceType=TRANSFORMPOSSOURCE_World}
    B=TransformPosition(None=@n12) {transformSourceType=TRANSFORMPOSSOURCE_World}
   )
   B=Multiply(A=Abs(None=@n13.RGB)) {constB=-1}
  )
  n15 = MaterialFunctionCall(Float2=ComponentMask(None=@n14) {b=true, r=true}) {materialFunction=&BreakOutFloat2Components}
  n16 = ScalarParameter() {defaultValue=1, parameterName=DebugSpeed}
  n17 = StaticSwitchParameter(
   True=LinearInterpolate(Alpha=MaterialFunctionCall(Speed=@n16) {materialFunction=&DebugTimeSine}.Result) {constA=0.01, constB=20}
   False=ScalarParameter() {defaultValue=5.7229, parameterName=FadeControl, sliderMax=40, sliderMin=0.01}
  ) {parameterName=Debug}
  n18 = Divide(A=Distance(A=Frac(None=@n15.R), B=ScalarParameter() {defaultValue=0.5, parameterName=U_Distance}), B=@n17)
  n19 = Divide(A=Distance(A=Frac(None=@n15.G), B=ScalarParameter() {defaultValue=0.5, parameterName=V_Distance}), B=@n17)
  n20 = MaterialFunctionCall(Float2=ComponentMask(None=@n14) {b=true, g=true}) {materialFunction=&BreakOutFloat2Components}
  n21 = Divide(A=Distance(A=Frac(None=@n20.R), B=ScalarParameter() {defaultValue=0.5, parameterName=U_Distance}), B=@n17)
  n22 = Divide(A=Distance(A=Frac(None=@n20.G), B=ScalarParameter() {defaultValue=0.5, parameterName=V_Distance}), B=@n17)
  n23 = Transform(None=VertexNormalWS()) {transformSourceType=TRANSFORMSOURCE_World, transformType=TRANSFORM_Local}
  n24 = ScalarParameter() {defaultValue=1, parameterName=Proj_Contrast}
  n25 = MaterialFunctionCall(In=Abs(None=ComponentMask(None=@n23) {r=true}), Contrast=@n24) {materialFunction=&CheapContrast}
  n26 = MaterialFunctionCall(Float2=ComponentMask(None=@n14) {g=true, r=true}) {materialFunction=&BreakOutFloat2Components}
  n27 = Distance(A=Frac(None=@n26.R), B=ScalarParameter() {defaultValue=0.5, parameterName=U_Distance})
  n28 = Distance(A=Frac(None=@n26.G), B=ScalarParameter() {defaultValue=0.5, parameterName=V_Distance})
  n29 = Abs(None=ComponentMask(None=@n23) {b=true})
  n30 = LinearInterpolate(
   A=LinearInterpolate(A=Max(A=@n18, B=@n19), B=Max(A=@n21, B=@n22), Alpha=ComponentMask(None=@n25.Result) {r=true}) {desc=Float3}
   B=Max(A=Divide(A=@n27, B=@n17), B=Divide(A=@n28, B=@n17))
   Alpha=ComponentMask(None=MaterialFunctionCall(In=@n29, Contrast=@n24) {materialFunction=&CheapContrast}.Result) {r=true}
  )
  n31 = OneMinus(None=MaterialFunctionCall(In=Clamp(None=@n30), Contrast=ScalarParameter() {defaultValue=1, parameterName=Exp}) {materialFunction=&CheapContrast}.Result)
  n32 = Constant() {r=1}
  n33 = If(
   A=Add(A=PixelDepth().R, B=ScalarParameter() {defaultValue=-0.7, parameterName=DepthBias})
   B=ComponentMask(None=SceneTexture() {sceneTextureId=PPI_CustomDepth}.Color) {r=true}
   "A > B"=Constant()
   "A == B"=@n32
   "A < B"=@n32
  )
  n34 = Constant() {r=1}
  n35 = TextureSample() {
   samplerSource=SSM_Wrap_WorldGroupSettings
   samplerType=SAMPLERTYPE_Normal
   texture=&T_Ashton_Boardwalk_N
  }
  n36 = Constant3Vector() {constant=<rgba 0 0 1 0>}
  n37 = ScalarParameter() {defaultValue=1.5, parameterName=Refraction}
  n38 = LinearInterpolate(Alpha=@n37) {constA=1, constAlpha=1, constB=2}
  n39 = Constant() {r=1}
  n40 = MakeMaterialAttributes!(
   BaseColor=LinearInterpolate(A=Multiply(A=@n1.RGB, B=@n2.RGB), B=Multiply(A=@n1.RGB, B=@n3), Alpha=Saturate(None=@n7))
   Metallic=@n8.G
   Specular=@n8.G
   Roughness=@n8.G
   Anisotropy
   EmissiveColor=@n10
   Opacity=Multiply(
    A=LinearInterpolate(Alpha=@n31) {constA=1, constB=0}
    B=ShadingPathSwitch(Default=@n33, Deferred=@n33, Mobile=@n34) {desc=Migrated from FeatureLevelSwitch}
   )
   OpacityMask
   Normal=QualitySwitch(
    Default=ShadingPathSwitch(Default=@n35.RGB, Deferred=@n35.RGB, Mobile=@n36) {desc=Migrated from FeatureLevelSwitch}
    Low=@n36
    High=@n35.RGB
    Medium=@n35.RGB
   )
   Tangent
   WorldPositionOffset
   SubsurfaceColor
   ClearCoat
   ClearCoatRoughness
   AmbientOcclusion
   Refraction=QualitySwitch(
    Default=ShadingPathSwitch(Default=@n38, Deferred=@n38, Mobile=@n39) {desc=Migrated from FeatureLevelSwitch}
    Low=@n39
    High=@n38
    Medium=LinearInterpolate(Alpha=@n37) {constA=1, constAlpha=1, constB=1.35}
   )
   CustomizedUVs_0
   CustomizedUVs_1
   CustomizedUVs_2
   CustomizedUVs_3
   CustomizedUVs_4
   CustomizedUVs_5
   CustomizedUVs_6
   CustomizedUVs_7
   PixelDepthOffset
   ShadingModel
   Displacement
  )
  MaterialFunctionCall!(In=@n40) {materialFunction=/Game/Effects/Niagara/MasterMaterials/Functions/MF_FX_EmissiveAdjustments_MA}
  n41 = TextureSample() {
   samplerSource=SSM_Wrap_WorldGroupSettings
   samplerType=SAMPLERTYPE_Masks
   texture=/BRCosmetics/Gadgets/Assets/VinderTech_GliderChute/Glider_Ashton_Boardwalk/Textures/T_Ashton_Boardwalk_FX
  }
  n42 = MaterialFunctionCall!(Alpha=@n41.G, isVehicle, Gmap, UVOffset, Local Normal, Local Position, EmissiveColorOverride, Material=@n40) {materialFunction=/Game/Weapons/WeaponSkins/MaterialFunction/PanningEmissive}
  SetMaterialAttributes(
   MaterialAttributes=@n40
   Emissive Color=Add(A=GetMaterialAttributes(None=@n42.Result) {attributeGetTypes [#1]}.Emissive Color, B=@n10)
  ) {attributeSetTypes [#1]}
 ]
}
